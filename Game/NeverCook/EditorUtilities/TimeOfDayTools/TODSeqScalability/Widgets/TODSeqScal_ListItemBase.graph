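# Editor list-item widget for the Time of Day sequence scalability tool, showing a searchable label and a grid of option checkboxes.
# On data update, it caches the item's filter map and option definitions, then adds a checkbox per matched option.
# Option category sets the row; a per-row count map sets the column.
# ClearOptions resets only that map, leaving grid children in place.
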WidgetBlueprint /Game/NeverCook/EditorUtilities/TimeOfDayTools/TODSeqScalability/Widgets/TODSeqScal_ListItemBase
graphs [
 AddGridOption {nodes [
   n1 = AddGridOption!() -> (then, Label:String, Enabled:bool=~, row:int, column:int) {is=title K2Node_FunctionEntry}
   > Set l_label(l_label:String=@n1) -> (Output_Get:String)
   > Set l_enabled(l_enabled:bool=@n1.Enabled) -> (Output_Get:bool)
   > Set l_row(l_row:int=@n1.row) -> (Output_Get:int=0)
   n2 = Find(
    TargetMap:Map of Integers to Integers=Get GridRowCountMap() -> (GridRowCountMap:Map of Integers to Integers)
    Key:"Integer (by ref)"=Get l_row() -> (l_row:int=0)
   ) -> (Value:int)
   > Set l_column(l_column:int=SelectInt(A=@n2, B=0, bPickA=@n2.ReturnValue)) -> (Output_Get:int=0)
   > n3 = Create Widget!(
    execute:Exec
    Class:User Widget Class Reference=&/TODSeqScal_ListItemCheckbox.TODSeqScal_ListItemCheckbox_C
    OwningPlayer:Player Controller Object Reference
   ) -> (then:Exec, ReturnValue:TODSeq Scal List Item Checkbox Object Reference)
   > Set l_checkbox(l_checkbox:TA Base Editor Button Widget Object Reference=@n3) -> (Output_Get:TA Base Editor Button Widget Object Reference)
   > n4 = Sequence!(execute) -> (then_0, then_1, then_2:Exec, then_3:Exec)
   > setSelected!(
    execute
    self:TA Base Editor Button Widget Object Reference=Get l_checkbox() -> (l_checkbox:TA Base Editor Button Widget Object Reference)
    selected:bool=Get l_enabled() -> (l_enabled:bool)
    throwEvent:bool=false
   ) -> (then)
   setLabel!(execute=@n4.then_1, self:TA Base Editor Button Widget Object Reference=Get l_checkbox(), label:String=Get l_label() -> (l_label:String)) -> (then)
   n5 = AddChildToGrid!(
    execute=@n4.then_2
    self:Grid Panel Object Reference=Get GridPanel_64() -> (GridPanel_64:Grid Panel Object Reference)
    Content:Widget Object Reference=Get l_checkbox()
    InRow:int=Get l_row() -> (l_row=0)
    InColumn:int=Get l_column() -> (l_column:int=0)
   ) -> (then, ReturnValue:Grid Slot Object Reference)
   > Set Padding!(execute, Padding:Margin Structure=Make Margin(Left=5, Top=3, Right=5, Bottom=3), self:Grid Slot Object Reference=@n5) -> (then, Output_Get:Margin Structure)
   IncrementRowMap!(execute=@n4.then_3, self, Key:"Integer (by ref)"=Get l_row() -> (l_row=0)) -> (then)
   int + int(A=@n2)
  ]}
 AddRowToGrid {nodes [
   n1 = AddRowToGrid!() -> (then, InputPin:Map of Strings to Booleans) {is=title K2Node_FunctionEntry}
   > n2 = Keys!(execute, TargetMap:Map of Strings to Booleans=@n1) -> (then, Keys:Array of Strings)
   > n3 = Values!(execute, TargetMap:Map of Strings to Booleans=@n1) -> (then, Values:Array of Booleans)
   > n4 = For Each Loop(Array:Array of Booleans=@n3) -> (Array Element:bool)
   n5 = Get CurrentGridRow() -> (CurrentGridRow:int=0)
   Increment Int(" "=@n4.Completed, Value=@n5)
   AddGridOption!(
    execute=@n4.LoopBody
    self
    Label:String="Get (a ref)"(Array:Array of Strings=@n2, Dimension 1=@n4.Array Index) -> (Output:"String (by ref)")
    Enabled:bool=@n4
    row:int=@n5
    column:int=@n4.Array Index
   ) -> (then) {is=title K2Node_CallFunction}
  ]}
 ClearOptions {nodes [
   ClearOptions() {is=title K2Node_FunctionEntry}
   > Clear!(execute, TargetMap:Map of Integers to Integers=Get GridRowCountMap()) -> (then) {is=title K2Node_CallFunction}
  ]}
 EventGraph {nodes [
   n1 = Event updateAfterDataChange!() -> (OutputDelegate, then, currentListItemData:Object Reference) {is=title K2Node_Event}
   > Parent: updateAfterDataChange!(execute, currentListItemData:Object Reference=@n1.currentListItemData) -> (then) {is=title K2Node_CallParentFunction}
   > n2 = Cast To TODSeqScalListItemData!(execute, Object=@n1.currentListItemData) -> (then, CastFailed, AsTODSeq Scal List Item Data:TODSeq Scal List Item Data Object Reference) {is=title K2Node_DynamicCast}
   > Set CurrentOptionData(CurrentOptionData:TODSeq Scal Seq Data Struct Structure=Get OptionData!(self:TODSeq Scal List Item Data Object Reference=@n2) -> (OptionData:TODSeq Scal Seq Data Struct Structure)) -> (Output_Get:TODSeq Scal Seq Data Struct Structure)
   > Set ScalabilityOptionData(ScalabilityOptionData:Map of Names to TODSeq Scal Option Struct Structures=Get ScalabilityOptionData!(self:TODSeq Scal List Item Data Object Reference=@n2) -> (ScalabilityOptionData:Map of Names to TODSeq Scal Option Struct Structures)) -> (Output_Get:Map of Names to TODSeq Scal Option Struct Structures)
   > ValidateOptionVisuals()
  ]}
 FilterBySearchString {nodes [
   n1 = FilterBySearchString!() -> (then, SearchString:String) {is=title K2Node_FunctionEntry}
   > Return Node!(execute, ShouldDisplay:bool=Contains(SearchIn=Get label() -> (label:String), Substring=@n1))
  ]}
 GetAdditionalSequenceData {nodes [
   GetAdditionalSequenceData() {is=title K2Node_FunctionEntry}
   n1 = Make Map!(
    Key 0:Name=IsVolumetricFogEnabled
    Value 0:bool=false
    Key 1:Name=QualityTODMIsLow
    Value 1:bool=true
    Key 2:Name=QualityTODMIsMed
    Value 2:bool=true
   ) -> (Map:Map of Names to Booleans) {is=title K2Node_MakeMap}
   n2 = Make Map(Key 0=IsLumenEnabled, Value 0=false, Key 1=QualityTODMIsMed, Value 1=true, Key 2=QualityTODMIsLow, Value 2=true)
   n3 = Make Map(Key 0=IsLumenEnabled, Value 0=false, Key 1=QualityTODMIsHigh, Value 1=true, Key 2=QualityTODMIsEpic, Value 2=true)
   > Return Node!(
    execute
    TrackNameDataMap:Map of Strings to TODSeq Scal Seq Data Struct Structures=Make Map(
     Key 0:String=Low End Fog Adjust
     Value 0:TODSeq Scal Seq Data Struct Structure=Make TODSeq Scal Seq Data Struct!(EnabledFilterMap_18_#1:Map of Names to Booleans=@n1) -> (TODSeqScal_SeqDataStruct:TODSeq Scal Seq Data Struct Structure) {is=title K2Node_MakeStruct}
     Key 1:String=Non Lumen Fill - High Quality
     Value 1:TODSeq Scal Seq Data Struct Structure=Make TODSeq Scal Seq Data Struct(EnabledFilterMap_18_#1=@n3)
     Key 2:String=Non Lumen Fill - Low Quality
     Value 2:TODSeq Scal Seq Data Struct Structure=Make TODSeq Scal Seq Data Struct(EnabledFilterMap_18_#1=@n2)
    ) -> (Map:Map of Strings to TODSeq Scal Seq Data Struct Structures)
   )
  ]}
 GetScalabilityOptionData {nodes [
   GetScalabilityOptionData() {is=title K2Node_FunctionEntry}
   > n1 = GetDataTableRowNames(Table=/Game/TimeOfDay/TODM_Sequence/Blueprint/Volume/Data/TODSeqScal_OptionTable)
   > n2 = For Each Loop(Array:Array of Names=@n1) -> (Array Element:Name)
   Return Node!(
    execute=@n2.Completed
    IdDataMap:Map of Names to TODSeq Scal Option Struct Structures=Get l_outMap() -> (l_outMap:Map of Names to TODSeq Scal Option Struct Structures)
   )
   > n3 = Get Data Table Row TODSeqScal_OptionTable!(
    execute:Exec
    DataTable:Data Table Object Reference=&TODSeqScal_OptionTable
    RowName:Name=@n2
   ) -> (then:Exec, RowNotFound:Exec, ReturnValue:TODSeq Scal Option Struct Structure) {is=title K2Node_GetDataTableRow}
   > Add!(
    execute
    TargetMap:Map of Names to TODSeq Scal Option Struct Structures=Get l_outMap()
    Key:"Name (by ref)"=Break TODSeq Scal Option Struct!(TODSeqScal_OptionStruct:"TODSeq Scal Option Struct Structure (by ref)"=@n3) -> (
     Name_6_#1:String
     Category_7_#2:TODSeqScal_Category_Enum Enum
     UniqueID_NeverChange_11_#3:Name
    ) {is=title K2Node_BreakStruct}.UniqueID_NeverChange_11_#3
    Value:"TODSeq Scal Option Struct Structure (by ref)"=@n3
   ) -> (then) {is=title K2Node_CallFunction}
  ]}
 IncrementRowMap {nodes [
   n1 = IncrementRowMap!() -> (then, Key:"Integer (by ref)") {is=title K2Node_FunctionEntry}
   n2 = Find(TargetMap=Get GridRowCountMap(), Key=@n1)
   > n3 = Branch(Condition=@n2.ReturnValue)
   > n4 = Values(TargetMap:Map of Integers to Integers=Get GridRowCountMap()) -> (Values:Array of Integers)
   > n5 = Increment Int(Value="Get (a ref)"(Array:Array of Integers=@n4, Dimension 1=@n2) -> (Output:"Integer (by ref)"))
   > Add(TargetMap:Map of Integers to Integers=Get GridRowCountMap(), Key:"Integer (by ref)"=@n1, Value:"Integer (by ref)"=@n5)
   Add(execute=@n3.else, TargetMap=Get GridRowCountMap(), Key=@n1, Value=1)
  ]}
 PopulateGridPanel {nodes [
   n1 = PopulateGridPanel!() -> (then, Features:Map of Strings to Booleans, Platforms:Map of Strings to Booleans, Qualities:Map of Strings to Booleans) {is=title K2Node_FunctionEntry}
   > AddRowToGrid!(execute, self, InputPin:Map of Strings to Booleans=@n1) -> (then) {is=title K2Node_CallFunction}
   > AddRowToGrid(InputPin=@n1.Platforms)
   > AddRowToGrid(InputPin=@n1.Qualities)
  ]}
 ValidateOptionVisuals {nodes [
   ValidateOptionVisuals() {is=title K2Node_FunctionEntry}
   > ClearOptions() {is=title K2Node_CallFunction}
   n1 = Break TODSeq Scal Seq Data Struct!(TODSeqScal_SeqDataStruct:"TODSeq Scal Seq Data Struct Structure (by ref)"=Get CurrentOptionData() -> (CurrentOptionData:TODSeq Scal Seq Data Struct Structure)) -> (EnabledFilterMap_18_#1:Map of Names to Booleans) {is=title K2Node_BreakStruct}
   > n2 = Keys(TargetMap:Map of Names to Booleans=@n1) -> (Keys:Array of Names)
   > n3 = Values(TargetMap:Map of Names to Booleans=@n1) -> (Values:Array of Booleans)
   > n4 = For Each Loop(Array:Array of Booleans=@n3) -> (Array Element:bool)
   n5 = Find(
    TargetMap:Map of Names to TODSeq Scal Option Struct Structures=Get ScalabilityOptionData!() -> (ScalabilityOptionData)
    Key:"Name (by ref)"="Get (a ref)"(Array:Array of Names=@n2, Dimension 1=@n4.Array Index) -> (Output:"Name (by ref)")
   ) -> (Value:TODSeq Scal Option Struct Structure)
   > Branch(Condition=@n5.ReturnValue)
   n6 = Break TODSeq Scal Option Struct(TODSeqScal_OptionStruct=@n5)
   > AddGridOption(
    Label=@n6
    Enabled=@n4
    row="To Integer (Byte)"!(InByte:Byte=@n6.Category_7_#2) -> (ReturnValue:int=0)
    column=0
   )
  ]}
 getBackgroundWidget {nodes [
   getBackgroundWidget() {is=title K2Node_FunctionEntry}
   > Return Node!(execute, backgroundWidget:Widget Object Reference=Get borderMain() -> (borderMain:Border Object Reference))
  ]}
 getLabelWidget {nodes [
   getLabelWidget() {is=title K2Node_FunctionEntry}
   > Return Node!(execute, labelWidget:Widget Object Reference=Get txtMain() -> (txtMain:Text Object Reference))
  ]}
 validateBackgroundColor {nodes [
   validateBackgroundColor() {is=title K2Node_FunctionEntry}
   > n1 = Parent: validateBackgroundColor!(execute) -> (then, handled:bool) {is=title K2Node_CallParentFunction}
   > Return Node!(execute, handled:bool=@n1)
  ]}
]
variables [
 CurrentGridRow {type=integer}
 CurrentOptionData {type=TODSeqScal_SeqDataStruct, default {enabledFilterMap {}}}
 GridRowCountMap {default {}}
 ScalabilityOptionData {type="map<TODSeqScal_OptionStruct>", default {}}
 SequenceData {type="map<TODSeqScal_SeqDataStruct>", default {}}
]
widgets {
 info {
  inheritedWidgetCount=0
  parentClass=/Game/Blueprints/Blutilities/ForTechArt/Widgets/Buttons/TA_BaseEditorButtonWidget.TA_BaseEditorButtonWidget_C
  rootWidgetClass=&Overlay
  widgetCount=9
 }
 widgets [
  Overlay_7 {widgetClassPath=&Overlay, widget=Overlay_7}
  VerticalBox_0 {
   widgetClassPath=&VerticalBox
   parent=Overlay_7
   widget=VerticalBox_0
   slot {class=OverlaySlot, properties {horizontalAlignment=HAlign_Fill}}
  }
  borderMain {
   widgetClassPath=&Border
   bIsVariable=true
   parent=VerticalBox_0
   widget=borderMain
   properties {padding=<ltrb 3 3 3 3>}
   slot {class=VerticalBoxSlot}
  }
  HorizontalBox_0 {
   widgetClassPath=&HorizontalBox
   parent=borderMain
   widget=HorizontalBox_0
   slot {class=BorderSlot, properties {padding=<ltrb 3 3 3 3>}}
  }
  ScaleBox_0 {
   widgetClassPath=&ScaleBox
   parent=HorizontalBox_0
   widget=ScaleBox_0
   properties {clipping=ClipToBounds, stretchDirection=DownOnly}
   slot {class=HorizontalBoxSlot, properties {padding=<ltrb 5 0 0 0>, size {sizeRule=Fill, value=1}}}
  }
  txtMain {
   widgetClassPath=&TextBlock
   bIsVariable=true
   parent=ScaleBox_0
   widget=txtMain
   properties {
    clipping=ClipToBounds
    justification=Center
    text=Text Block
    colorAndOpacity {colorUseRule=UseColor_Specified, specifiedColor=<rgba 0.0104 0.0104 0.0104 1>}
    font {
     bForceMonospaced=false
     bMaterialIsStencil=false
     fontMaterial=null
     fontObject=&Roboto
     letterSpacing=0
     monospacedWidth=1
     size=10
     skewAmount=0
     typefaceFontName=Bold
     outlineSettings {
      bApplyOutlineToDropShadows=false
      bMiteredCorners=false
      bSeparateFillAlpha=false
      outlineMaterial=null
      outlineSize=0
      outlineColor=<rgba 0 0 0 1>
     }
    }
    shadowColorAndOpacity=<rgba 0 0 0 0.2>
   }
   slot {class=ScaleBoxSlot, properties {horizontalAlignment=HAlign_Fill}}
  }
  GridPanel_64 {
   widgetClassPath=&GridPanel
   bIsVariable=true
   parent=HorizontalBox_0
   widget=GridPanel_64
   slot {class=HorizontalBoxSlot, properties {size {sizeRule=Fill, value=1}}}
  }
  TA_Generic01_Button {
   widgetClassPath=&TA_Generic01_Button
   parent=HorizontalBox_0
   widget=TA_Generic01_Button
   properties {iconBrush {
     bIsDynamicallyLoaded=false
     drawAs=Image
     imageType=NoImage
     mirroring=NoMirror
     resourceName=None
     resourceObject=/Game/Athena/UI/Frontend/Parties/Art/T_UI_PlusSign_64x
     tiling=NoTile
     imageSize=<12 12>
     margin=<ltrb 0 0 0 0>
     outlineSettings {
      bUseBrushTransparency=false
      roundingType=HalfHeightRadius
      width=0
      color {colorUseRule=UseColor_Specified, specifiedColor=<rgba 0 0 0 0>}
      cornerRadii=<0 0 0 1>
     }
     tintColor {colorUseRule=UseColor_Specified, specifiedColor=<rgba 1 1 1 1>}
     uVRegion {bIsValid=false, max=<0 0>, min=<0 0>}
    }}
   slot {class=HorizontalBoxSlot, properties {horizontalAlignment=HAlign_Right, verticalAlignment=VAlign_Center}}
  }
  imageEdge {
   widgetClassPath=&Image
   bIsVariable=true
   parent=VerticalBox_0
   widget=imageEdge
   properties {
    brush {
     bIsDynamicallyLoaded=false
     drawAs=Image
     imageType=NoImage
     mirroring=NoMirror
     resourceName=None
     resourceObject=null
     tiling=NoTile
     imageSize=<32 2>
     margin=<ltrb 0 0 0 0>
     outlineSettings {
      bUseBrushTransparency=false
      roundingType=HalfHeightRadius
      width=0
      color {colorUseRule=UseColor_Specified, specifiedColor=<rgba 0 0 0 0>}
      cornerRadii=<0 0 0 0>
     }
     tintColor {colorUseRule=UseColor_Specified, specifiedColor=<rgba 1 1 1 1>}
     uVRegion {bIsValid=false, max=<0 0>, min=<0 0>}
    }
    colorAndOpacity=<rgba 0.2031 0.2031 0.2031 1>
   }
   slot {class=VerticalBoxSlot}
  }
 ]
}
