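# Editor utility widget that lists Time of Day sequence scalability presets, such as non-volumetric fog, non-Lumen fill and mobile fog.
# On Construct, it loads option definitions from TODSeqScal_OptionTable into a map keyed by unique ID.
# Presets are hard-coded name-to-filter maps of scalability flags.
# Each becomes a list item data object carrying the option map; an unconnected chain of generic placeholder items remains in the graph.
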EditorUtilityWidgetBlueprint /Game/NeverCook/EditorUtilities/TimeOfDayTools/TODSeqScalability/EUW_TODSeqScalability
graphs [
 AddSequenceItem {nodes [
   n1 = AddSequenceItem!() -> (
    then
    OptionData:TODSeq Scal Seq Data Struct Structure
    displayString:String
    ScalabilityOptionData:Map of Names to TODSeq Scal Option Struct Structures
   ) {is=title K2Node_FunctionEntry}
   > n2 = Construct Object from Class(Class=/Game/NeverCook/EditorUtilities/TimeOfDayTools/TODSeqScalability/Data/TODSeqScalListItemData.TODSeqScalListItemData_C) -> (ReturnValue:TODSeq Scal List Item Data Object Reference)
   > Set OptionData!(execute, OptionData:TODSeq Scal Seq Data Struct Structure=@n1, self:TODSeq Scal List Item Data Object Reference=@n2) -> (then, Output_Get:TODSeq Scal Seq Data Struct Structure)
   > Set displayString!(execute, displayString:String=@n1.displayString, self:TA Generic List Item Data Object Reference=@n2) -> (then, Output_Get:String)
   > Set ScalabilityOptionData!(
    execute
    ScalabilityOptionData:Map of Names to TODSeq Scal Option Struct Structures=@n1.ScalabilityOptionData
    self:TODSeq Scal List Item Data Object Reference=@n2
   ) -> (then, Output_Get:Map of Names to TODSeq Scal Option Struct Structures)
   > addItem!(
    execute
    self:TA Generic 01 List Object Reference=Get TA_Generic01_List() -> (TA_Generic01_List:TA Generic 01 List Object Reference)
    dataObject:Object Reference=@n2
   ) -> (then, newItem:User Widget Object Reference, newData:Object Reference)
  ]}
 EventGraph {nodes [
   n1 = Get TA_Generic01_List()
   addItemGeneric!(execute, self:TA Generic 01 List Object Reference=@n1, displayString:String=No Lumen Fill) -> (then, widget:User Widget Object Reference, data:Object Reference)
   > addItemGeneric(self=@n1, displayString=Low End Fog)
   > addItemGeneric(self=@n1, displayString=Low End Bloom)
   > addItemGeneric(self=@n1, displayString=Volume Fog Balance)
   Event Construct()
   > n2 = GetScalabilityOptionData!(execute, self) -> (then, IdDataMap:Map of Names to TODSeq Scal Option Struct Structures)
   > n3 = GetAdditionalSequenceData!(execute, self) -> (then, TrackNameDataMap:Map of Strings to TODSeq Scal Seq Data Struct Structures)
   > n4 = Keys!(execute, TargetMap:Map of Strings to TODSeq Scal Seq Data Struct Structures=@n3) -> (then, Keys:Array of Strings)
   > n5 = Values!(execute, TargetMap:Map of Strings to TODSeq Scal Seq Data Struct Structures=@n3) -> (then, Values:Array of TODSeq Scal Seq Data Struct Structures)
   > n6 = For Each Loop(Array:Array of TODSeq Scal Seq Data Struct Structures=@n5) -> (Array Element:TODSeq Scal Seq Data Struct Structure)
   > AddSequenceItem!(
    execute
    self
    OptionData:TODSeq Scal Seq Data Struct Structure=@n6
    displayString:String="Get (a ref)"(Array:Array of Strings=@n4, Dimension 1=@n6.Array Index) -> (Output:"String (by ref)")
    ScalabilityOptionData:Map of Names to TODSeq Scal Option Struct Structures=@n2
   ) -> (then) {is=title K2Node_CallFunction}
  ]}
 GetAdditionalSequenceData {nodes [
   GetAdditionalSequenceData() {is=title K2Node_FunctionEntry}
   n1 = Make Map!(
    Key 0:Name=IsVolumetricFogEnabled
    Value 0:bool=false
    Key 1:Name=QualityTODMIsLow
    Value 1:bool=true
    Key 2:Name=QualityTODMIsMed
    Value 2:bool=true
   ) -> (Map:Map of Names to Booleans) {is=title K2Node_MakeMap}
   n2 = Make Map!(Key 0=IsPlatformPCPerfMode, Value 0=true) -> (Map)
   n3 = Make Map!(Key 0=IsLumenEnabled, Value 0=false, Key 1:Name=QualityTODMIsHigh, Value 1:bool=true, Key 2:Name=QualityTODMIsEpic, Value 2:bool=true) -> (Map)
   n4 = Make Map(Key 0=IsLumenEnabled, Value 0=false, Key 1=QualityTODMIsMed, Value 1=true, Key 2=QualityTODMIsLow, Value 2=true)
   n5 = Make Map!(Key 0=IsMobileRendered, Value 0=true, Key 1=IsPlatformPCPerfMode, Value 1=false) -> (Map)
   > Return Node!(
    execute
    TrackNameDataMap:Map of Strings to TODSeq Scal Seq Data Struct Structures=Make Map!(
     Key 0:String=Non Volumetric Fog
     Value 0:TODSeq Scal Seq Data Struct Structure=Make TODSeq Scal Seq Data Struct!(EnabledFilterMap_18_#1:Map of Names to Booleans=@n1) -> (TODSeqScal_SeqDataStruct:TODSeq Scal Seq Data Struct Structure) {is=title K2Node_MakeStruct}
     Key 1:String=Non Lumen Fill - High Quality
     Value 1:TODSeq Scal Seq Data Struct Structure=Make TODSeq Scal Seq Data Struct(EnabledFilterMap_18_#1=@n3)
     Key 2:String=Non Lumen Fill - Low Quality
     Value 2:TODSeq Scal Seq Data Struct Structure=Make TODSeq Scal Seq Data Struct(EnabledFilterMap_18_#1=@n4)
     Key 3:String=Mobile Fog
     Value 3:TODSeq Scal Seq Data Struct Structure=Make TODSeq Scal Seq Data Struct(EnabledFilterMap_18_#1=@n5)
     Key 4:String=PC Perf Fog
     Value 4:TODSeq Scal Seq Data Struct Structure=Make TODSeq Scal Seq Data Struct(EnabledFilterMap_18_#1=@n2)
    ) -> (Map:Map of Strings to TODSeq Scal Seq Data Struct Structures)
   )
  ]}
 GetScalabilityOptionData {nodes [
   GetScalabilityOptionData() {is=title K2Node_FunctionEntry}
   > n1 = GetDataTableRowNames(Table=/Game/TimeOfDay/TODM_Sequence/Blueprint/Volume/Data/TODSeqScal_OptionTable)
   > n2 = For Each Loop(Array:Array of Names=@n1) -> (Array Element:Name)
   Return Node!(
    execute=@n2.Completed
    IdDataMap:Map of Names to TODSeq Scal Option Struct Structures=Get l_outMap() -> (l_outMap:Map of Names to TODSeq Scal Option Struct Structures)
   )
   > n3 = Get Data Table Row TODSeqScal_OptionTable!(
    execute:Exec
    DataTable:Data Table Object Reference=&TODSeqScal_OptionTable
    RowName:Name=@n2
   ) -> (then:Exec, RowNotFound:Exec, ReturnValue:TODSeq Scal Option Struct Structure) {is=title K2Node_GetDataTableRow}
   > Add!(
    execute
    TargetMap:Map of Names to TODSeq Scal Option Struct Structures=Get l_outMap()
    Key:"Name (by ref)"=Break TODSeq Scal Option Struct!(TODSeqScal_OptionStruct:"TODSeq Scal Option Struct Structure (by ref)"=@n3) -> (
     Name_6_#1:String
     Category_7_#2:TODSeqScal_Category_Enum Enum
     UniqueID_NeverChange_11_#3:Name
    ) {is=title K2Node_BreakStruct}.UniqueID_NeverChange_11_#3
    Value:"TODSeq Scal Option Struct Structure (by ref)"=@n3
   ) -> (then) {is=title K2Node_CallFunction}
  ]}
]
variables []
widgets {
 info {rootWidgetClass=&StackBox, widgetCount=2}
 widgets [
  StackBox_0 {widgetClassPath=&StackBox, widget=StackBox_0}
  TA_Generic01_List {
   widgetClassPath=/Game/Blueprints/Blutilities/ForTechArt/Widgets/List/TA_Generic01_List
   bIsVariable=true
   parent=StackBox_0
   widget=TA_Generic01_List
   properties {
    allowMultiSelect=false
    allowNoSelection=true
    enableSelection=false
    itemPadding=15
    listItemClass=/Game/NeverCook/EditorUtilities/TimeOfDayTools/TODSeqScalability/Widgets/TODSeqScal_ListItemBase.TODSeqScal_ListItemBase_C
   }
   slot {class=StackBoxSlot, properties {size {sizeRule=Fill, value=1}}}
  }
 ]
}
